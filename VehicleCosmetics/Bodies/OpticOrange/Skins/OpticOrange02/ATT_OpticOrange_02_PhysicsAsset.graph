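# Physics asset for the ATT_OpticOrange_02 part of the OpticOrange02 skin of the OpticOrange vehicle body cosmetic.
# It holds one body on the root bone, with no collision shapes listed for it.
# It has no constraints, so it defines no joints or ragdoll chain.
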PhysicsAsset /VehicleCosmetics/Bodies/OpticOrange/Skins/OpticOrange02/ATT_OpticOrange_02_PhysicsAsset
bodies {root []}
constraints []
